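# Physics asset for the ribbon face accessory of the female medium Rumble player body, defining two dynamic strap chains that hang from the head.
# Each side is a chain of capsule bodies from a base bone to the tip, with narrow swing and twist limits near the root and wider swing further down.
# Capsules on head, spine and pelvis give the straps body volume, linked by free-rotation constraints, as are the strap bases to the head.
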PhysicsAsset /BRCosmetics/Characters/Player/Female/Medium/Bodies/F_MED_Rumble/Meshes/Parts/F_MED_Rumble_FaceAcc_Ribbons_Dyn
bodies {
 dyn_hair_strap_a_base_l [{shapeType=Capsule, length=0.505, radius=0.505, rotation=<rot 0 -90 0>}]
 dyn_hair_strap_a_base_r [{shapeType=Capsule, length=0.505, radius=0.505, rotation=<rot 0 -90 0>}]
 dyn_hair_strap_a_l [{shapeType=Capsule, length=1.75, radius=1.5, center=<3 0 0>, rotation=<rot 90 0 0>}]
 dyn_hair_strap_a_r [{shapeType=Capsule, length=1.625, radius=1.5, center=<3 0 0>, rotation=<rot 90 0 0>}]
 dyn_hair_strap_b_l [{shapeType=Capsule, length=4.375, radius=1.875, center=<3 0 0>, rotation=<rot 90 0 0>}]
 dyn_hair_strap_b_r [{shapeType=Capsule, length=4.5, radius=1.875, center=<3 0 0>, rotation=<rot 90 0 0>}]
 dyn_hair_strap_c_l [{shapeType=Capsule, length=5.75, radius=1.875, center=<4 0 0>, rotation=<rot 90 0 0>}]
 dyn_hair_strap_c_r [{shapeType=Capsule, length=5.625, radius=1.875, center=<4 0 0>, rotation=<rot 90 0 0>}]
 dyn_hair_strap_d_l [{shapeType=Capsule, length=5.75, radius=1.875, center=<5 0 0>, rotation=<rot 90 0 0>}]
 dyn_hair_strap_d_r [{shapeType=Capsule, length=7.375, radius=1.875, center=<5 0 0>, rotation=<rot 90 0 0>}]
 dyn_hair_strap_e_l [{shapeType=Capsule, length=10.875, radius=1.875, center=<6 0 0>, rotation=<rot 90 0 0>}]
 dyn_hair_strap_e_r [{shapeType=Capsule, length=10.5, radius=1.875, center=<6 0 0>, rotation=<rot 90 0 0>}]
 dyn_hair_strap_g_l [{shapeType=Capsule, length=10.875, radius=1.875, center=<6 0 0>, rotation=<rot 90 0 0>}]
 dyn_hair_strap_g_r [{shapeType=Capsule, length=10.75, radius=1.875, center=<6 0 0>, rotation=<rot 90 0 0>}]
 dyn_hair_strap_l_l [{shapeType=Capsule, length=10.875, radius=1.875, center=<6 0 0>, rotation=<rot 90 0 0>}]
 dyn_hair_strap_r_r [{shapeType=Capsule, length=10.75, radius=1.875, center=<6 0 0>, rotation=<rot 90 0 0>}]
 head [{shapeType=Capsule, length=8.625, radius=9.375, center=<4.2261 1.7749 0>, rotation=<rot 90 0 -4.9998>}]
 pelvis [{shapeType=Capsule, length=15.5, radius=22.625, center=<-8.7896 -0.818 -0.0001>}]
 spine_02 [{shapeType=Capsule, length=16.75, radius=13, center=<4.3254 0.8794 -0.0001>}]
 spine_04 [{shapeType=Capsule, length=27.375, radius=14.375, center=<2.4988 0.4771 -0.0001>}]
 spine_05 [{shapeType=Capsule, length=37.375, radius=12.375, center=<0.4772 0.2204 -0.0001>}]
}
constraints [
 {
  bone1Name=dyn_hair_strap_a_base_l
  bone2Name=head
  swing1LimitDegrees=45
  swing1Motion=Free
  swing2LimitDegrees=45
  swing2Motion=Free
  twistLimitDegrees=45
  twistMotion=Free
 }
 {
  bone1Name=dyn_hair_strap_a_base_r
  bone2Name=head
  swing1LimitDegrees=45
  swing1Motion=Free
  swing2LimitDegrees=45
  swing2Motion=Free
  twistLimitDegrees=45
  twistMotion=Free
 }
 {bone1Name=dyn_hair_strap_a_l, bone2Name=dyn_hair_strap_a_base_l, swing1LimitDegrees=10, swing2LimitDegrees=25, twistLimitDegrees=10}
 {bone1Name=dyn_hair_strap_b_l, bone2Name=dyn_hair_strap_a_l, swing1LimitDegrees=10, swing2LimitDegrees=25, twistLimitDegrees=10}
 {bone1Name=dyn_hair_strap_c_l, bone2Name=dyn_hair_strap_b_l, swing1LimitDegrees=10, swing2LimitDegrees=25, twistLimitDegrees=10}
 {bone1Name=dyn_hair_strap_d_l, bone2Name=dyn_hair_strap_c_l, swing1LimitDegrees=20, swing2LimitDegrees=65, twistLimitDegrees=10}
 {bone1Name=dyn_hair_strap_e_l, bone2Name=dyn_hair_strap_d_l, swing1LimitDegrees=20, swing2LimitDegrees=65, twistLimitDegrees=10}
 {bone1Name=dyn_hair_strap_a_r, bone2Name=dyn_hair_strap_a_base_r, swing1LimitDegrees=10, swing2LimitDegrees=25, twistLimitDegrees=10}
 {bone1Name=dyn_hair_strap_b_r, bone2Name=dyn_hair_strap_a_r, swing1LimitDegrees=10, swing2LimitDegrees=25, twistLimitDegrees=10}
 {bone1Name=dyn_hair_strap_c_r, bone2Name=dyn_hair_strap_b_r, swing1LimitDegrees=10, swing2LimitDegrees=25, twistLimitDegrees=10}
 {bone1Name=dyn_hair_strap_d_r, bone2Name=dyn_hair_strap_c_r, swing1LimitDegrees=20, swing2LimitDegrees=65, twistLimitDegrees=10}
 {bone1Name=dyn_hair_strap_e_r, bone2Name=dyn_hair_strap_d_r, swing1LimitDegrees=20, swing2LimitDegrees=65, twistLimitDegrees=10}
 {
  bone1Name=head
  bone2Name=spine_05
  swing1LimitDegrees=45
  swing1Motion=Free
  swing2LimitDegrees=45
  swing2Motion=Free
  twistLimitDegrees=45
  twistMotion=Free
 }
 {
  bone1Name=spine_05
  bone2Name=spine_04
  swing1LimitDegrees=45
  swing1Motion=Free
  swing2LimitDegrees=45
  swing2Motion=Free
  twistLimitDegrees=45
  twistMotion=Free
 }
 {bone1Name=dyn_hair_strap_l_l, bone2Name=dyn_hair_strap_e_l, swing1LimitDegrees=20, swing2LimitDegrees=65, twistLimitDegrees=10}
 {bone1Name=dyn_hair_strap_g_l, bone2Name=dyn_hair_strap_l_l, swing1LimitDegrees=20, swing2LimitDegrees=65, twistLimitDegrees=10}
 {bone1Name=dyn_hair_strap_r_r, bone2Name=dyn_hair_strap_e_r, swing1LimitDegrees=20, swing2LimitDegrees=65, twistLimitDegrees=10}
 {bone1Name=dyn_hair_strap_g_r, bone2Name=dyn_hair_strap_r_r, swing1LimitDegrees=20, swing2LimitDegrees=65, twistLimitDegrees=10}
 {
  bone1Name=spine_04
  bone2Name=spine_02
  swing1LimitDegrees=45
  swing1Motion=Free
  swing2LimitDegrees=45
  swing2Motion=Free
  twistLimitDegrees=45
  twistMotion=Free
 }
 {
  bone1Name=spine_02
  bone2Name=pelvis
  swing1LimitDegrees=45
  swing1Motion=Free
  swing2LimitDegrees=45
  swing2Motion=Free
  twistLimitDegrees=45
  twistMotion=Free
 }
]
