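# Animation Blueprint for the F_MED_Rumble_Red face accessory, adding secondary motion to hair tips and straps.
# A cached base pose copies the parent mesh, simulates tips with AnimDynamics and aims straps at the spine.
# RigidBody branches swap physics assets for no backbling, backbling or cape, with a look-at and trail-controller fallback.
# The EventGraph drives noise rotations, gravity, movement and head-pitch values, and detects Rumble backblings by mesh name.
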AnimBlueprint /BRCosmetics/Characters/Player/Female/Medium/Bodies/F_MED_Rumble_Red/Meshes/Parts/F_MED_Rumble_Red_FaceAcc_AnimBP
target_skeleton=&/F_MED_Rumble_Red_FaceAcc_Skeleton
graphs [
 AnimGraph {nodes [
   n1 = "Transform (Modify) Bone - Bone: dyn_hair_strap_a_base_l"!(
    ComponentPose=Local To Component(LocalPose=Use cached pose 'PerPose'())
    bAlphaBoolEnabled:bool=True
    Translation="0,0,0"
    Rotation="15,0,0"
    Scale="1,1,1"
   ) -> (Pose) {
    alphaInputType=Bool
    rotationMode=BMM_Additive
    rotationSpace=BCS_BoneSpace
    alphaBoolBlend {blendOption=Linear}
    boneToModify {boneName=dyn_hair_strap_a_base_l}
    rotation {pitch=15}
   }
   n2 = "Transform (Modify) Bone - Bone: dyn_hair_strap_a_base_r"!(ComponentPose=@n1, bAlphaBoolEnabled:bool=True, Translation="0,0,0", Rotation="15,0,0", Scale="1,1,1") -> (Pose) {
    alphaInputType=Bool
    rotationMode=BMM_Additive
    rotationSpace=BCS_BoneSpace
    alphaBoolBlend {blendOption=Linear}
    boneToModify {boneName=dyn_hair_strap_a_base_r}
    rotation {pitch=15}
   }
   n3 = Get isRumbleBackbling() -> (isRumbleBackbling:bool)
   n4 = "Transform (Modify) Bone - Bone: dyn_hair_strap_a_base_l"(ComponentPose=@n2, bAlphaBoolEnabled=@n3, Translation="0,0,0", Rotation="15,0,0", Scale="1,1,1") {
    alphaInputType=Bool
    rotationMode=BMM_Additive
    rotationSpace=BCS_BoneSpace
    alphaBoolBlend {blendOption=Linear}
    boneToModify {boneName=dyn_hair_strap_a_base_l}
    rotation {pitch=15}
   }
   n5 = "Transform (Modify) Bone - Bone: dyn_hair_strap_a_base_r"(ComponentPose=@n4, bAlphaBoolEnabled=@n3, Translation="0,0,0", Rotation="15,0,0", Scale="1,1,1") {
    alphaInputType=Bool
    rotationMode=BMM_Additive
    rotationSpace=BCS_BoneSpace
    alphaBoolBlend {blendOption=Linear}
    boneToModify {boneName=dyn_hair_strap_a_base_r}
    rotation {pitch=15}
   }
   n6 = "Transform (Modify) Bone - Bone: dyn_hair_strap_a_base_l"(ComponentPose=@n5, bAlphaBoolEnabled=True, Translation="0,0,0", Rotation=Get HairRotator_03() -> (HairRotator_03:Rotator), Scale="1,1,1") {
    alphaInputType=Bool
    rotationMode=BMM_Additive
    rotationSpace=BCS_BoneSpace
    alphaBoolBlend {blendOption=Linear}
    boneToModify {boneName=dyn_hair_strap_a_base_l}
   }
   n7 = "Transform (Modify) Bone - Bone: dyn_hair_strap_a_base_r"(ComponentPose=@n6, bAlphaBoolEnabled=True, Translation="0,0,0", Rotation=Get HairRotator_a_04() -> (HairRotator_a_04:Rotator), Scale="1,1,1") {
    alphaInputType=Bool
    rotationMode=BMM_Additive
    rotationSpace=BCS_BoneSpace
    alphaBoolBlend {blendOption=Linear}
    boneToModify {boneName=dyn_hair_strap_a_base_r}
   }
   n8 = "Transform (Modify) Bone - Bone: dyn_backpack_collider"!(ComponentPose=@n7, bAlphaBoolEnabled:bool=Get isRumbleBackbling(), Translation="0,-11,1.4", Rotation="90,0,0", Scale="1,1,1") -> (Pose) {
    alphaInputType=Bool
    rotationMode=BMM_Additive
    rotationSpace=BCS_BoneSpace
    translationMode=BMM_Additive
    translationSpace=BCS_BoneSpace
    alphaBoolBlend {blendOption=Linear}
    boneToModify {boneName=dyn_backpack_collider}
    rotation {pitch=90}
    translation {y=-11, z=1.4}
   }
   n9 = Get AccScaleOverride() -> (AccScaleOverride:Vector)
   n10 = Get GravityOverride()
   n11 = RigidBody!(ComponentPose=@n8, bAlphaBoolEnabled:bool, OverrideWorldGravity:Vector=@n10, ExternalForce, ComponentLinearAccScale:Vector=@n9) -> (Pose) {
    alphaInputType=Bool
    bClampLinearTranslationLimitToRefPose=true
    bOverrideWorldGravity=true
    overridePhysicsAsset=/BRCosmetics/Characters/Player/Female/Medium/Bodies/F_MED_Rumble/Meshes/Parts/F_MED_Rumble_FaceAcc_Ribbons_Dyn_Backbling
    simulationSpace=BaseBoneSpace
    alphaBoolBlend {blendOption=Linear}
    baseBoneRef {boneName=head}
    componentAppliedLinearAccClamp=<100 300 300>
    componentLinearAccScale=<0.1 0.2 0.2>
    componentLinearVelScale=<0.1 0.2 0.2>
    overrideWorldGravity {z=-980}
    simSpaceSettings {
     maxAngularAcceleration=200
     maxAngularVelocity=3
     maxLinearAcceleration=500
     maxLinearVelocity=100
     velocityScaleZ=0
     externalLinearDragV=<0.5 0.5>
    }
   }
   n12 = "Transform (Modify) Bone - Bone: dyn_hair_strap_a_base_l"(
    ComponentPose=Local To Component(LocalPose=Use cached pose 'PerPose'())
    bAlphaBoolEnabled=True
    Translation="0,0,0"
    Rotation="8,0,0"
    Scale="1,1,1"
   ) {
    alphaInputType=Bool
    rotationMode=BMM_Additive
    rotationSpace=BCS_BoneSpace
    alphaBoolBlend {blendOption=Linear}
    boneToModify {boneName=dyn_hair_strap_a_base_l}
    rotation {pitch=8}
   }
   n13 = "Transform (Modify) Bone - Bone: dyn_hair_strap_a_base_r"(ComponentPose=@n12, bAlphaBoolEnabled=True, Translation="0,0,0", Rotation="8,0,0", Scale="1,1,1") {
    alphaInputType=Bool
    rotationMode=BMM_Additive
    rotationSpace=BCS_BoneSpace
    alphaBoolBlend {blendOption=Linear}
    boneToModify {boneName=dyn_hair_strap_a_base_r}
    rotation {pitch=8}
   }
   n14 = "Transform (Modify) Bone - Bone: dyn_hair_strap_a_base_l"!(ComponentPose=@n13, Alpha:float=0.8, Translation="0,0,0", Rotation=Get HairRotator_01() -> (HairRotator_01:Rotator), Scale="1,1,1") -> (Pose) {rotationMode=BMM_Additive, rotationSpace=BCS_BoneSpace, alphaBoolBlend {blendOption=Linear}, boneToModify {boneName=dyn_hair_strap_a_base_l}}
   n15 = "Transform (Modify) Bone - Bone: dyn_hair_strap_a_base_r"!(ComponentPose=@n14, Alpha:float=0.8, Translation="0,0,0", Rotation=Get HairRotator_a_02() -> (HairRotator_a_02:Rotator), Scale="1,1,1") -> (Pose) {rotationMode=BMM_Additive, rotationSpace=BCS_BoneSpace, alphaBoolBlend {blendOption=Linear}, boneToModify {boneName=dyn_hair_strap_a_base_r}}
   n16 = RigidBody(ComponentPose=@n15, OverrideWorldGravity=@n10, ComponentLinearAccScale=@n9) {
    alphaInputType=Bool
    bClampLinearTranslationLimitToRefPose=true
    bOverrideWorldGravity=true
    overridePhysicsAsset=/BRCosmetics/Characters/Player/Female/Medium/Bodies/F_MED_Rumble/Meshes/Parts/F_MED_Rumble_FaceAcc_Ribbons_Dyn_WearCape
    simulationSpace=BaseBoneSpace
    alphaBoolBlend {blendOption=Linear}
    baseBoneRef {boneName=head}
    componentAppliedLinearAccClamp=<200 300 300>
    componentLinearAccScale=<0.1 0.2 0.2>
    componentLinearVelScale=<0.5 0.2 0.2>
    overrideWorldGravity {z=-980}
    simSpaceSettings {
     maxAngularAcceleration=200
     maxAngularVelocity=3
     maxLinearAcceleration=500
     maxLinearVelocity=100
     velocityScaleZ=0
     externalLinearDragV=<0.5 0.5>
    }
   }
   n17 = Blend Poses by bool(
    BlendPose_0=Component To Local(ComponentPose=@n16)
    BlendPose_1=Component To Local(ComponentPose=@n11)
    BlendTime_0=0.3
    BlendTime_1=0.3
    bActiveValue=Get isWearCape() -> (isWearCape:bool)
   ) {blendType=Linear}
   n18 = "Transform (Modify) Bone - Bone: dyn_hair_strap_a_base_l"!(
    ComponentPose=Local To Component(LocalPose=Use cached pose 'PerPose'())
    bAlphaBoolEnabled:bool=True
    Translation="0,0,0"
    Rotation=Get HairRotator_01()
    Scale="1,1,1"
   ) -> (Pose) {
    alphaInputType=Bool
    rotationMode=BMM_Additive
    rotationSpace=BCS_BoneSpace
    alphaBoolBlend {blendOption=Linear}
    boneToModify {boneName=dyn_hair_strap_a_base_l}
   }
   n19 = "Transform (Modify) Bone - Bone: dyn_hair_strap_a_base_r"!(ComponentPose=@n18, bAlphaBoolEnabled:bool=True, Translation="0,0,0", Rotation=Get HairRotator_a_02(), Scale="1,1,1") -> (Pose) {
    alphaInputType=Bool
    rotationMode=BMM_Additive
    rotationSpace=BCS_BoneSpace
    alphaBoolBlend {blendOption=Linear}
    boneToModify {boneName=dyn_hair_strap_a_base_r}
   }
   n20 = RigidBody(ComponentPose=@n19, OverrideWorldGravity=@n10, ComponentLinearAccScale=@n9) {
    alphaInputType=Bool
    bClampLinearTranslationLimitToRefPose=true
    bOverrideWorldGravity=true
    overridePhysicsAsset=/BRCosmetics/Characters/Player/Female/Medium/Bodies/F_MED_Rumble/Meshes/Parts/F_MED_Rumble_FaceAcc_Ribbons_Dyn
    simulationSpace=BaseBoneSpace
    alphaBoolBlend {blendOption=Linear}
    baseBoneRef {boneName=head}
    componentAppliedLinearAccClamp=<200 350 350>
    componentLinearAccScale=<0.1 0.2 0.2>
    componentLinearVelScale=<0.5 0.2 0.2>
    overrideWorldGravity {z=-980}
    simSpaceSettings {
     maxAngularAcceleration=200
     maxAngularVelocity=3
     maxLinearAcceleration=500
     maxLinearVelocity=100
     velocityScaleZ=0
     externalLinearDragV=<0.5 0.5>
    }
   }
   n21 = Component To Local(ComponentPose=@n20)
   n22 = Get isWearBackbling() -> (isWearBackbling:bool)
   n23 = Get LookUp() -> (LookUp:double)
   n24 = "Transform (Modify) Bone - Bone: dyn_hair_strap_a_base_l"!(
    ComponentPose=Local To Component(LocalPose=Use cached pose 'PerPose'())
    Alpha:float=@n23
    Translation="0,0,0"
    Rotation="-10,0,0"
    Scale="1,1,1"
   ) -> (Pose) {
    rotationMode=BMM_Additive
    rotationSpace=BCS_BoneSpace
    alphaBoolBlend {blendOption=Linear}
    boneToModify {boneName=dyn_hair_strap_a_base_l}
    rotation {pitch=-10}
   }
   n25 = "Transform (Modify) Bone - Bone: dyn_hair_strap_a_base_r"!(ComponentPose=@n24, Alpha:float=@n23, Translation="0,0,0", Rotation="-10,0,0", Scale="1,1,1") -> (Pose) {
    rotationMode=BMM_Additive
    rotationSpace=BCS_BoneSpace
    alphaBoolBlend {blendOption=Linear}
    boneToModify {boneName=dyn_hair_strap_a_base_r}
    rotation {pitch=-10}
   }
   n26 = Get LookDown() -> (LookDown:double)
   n27 = "Transform (Modify) Bone - Bone: dyn_hair_strap_a_base_l"(ComponentPose=@n25, Alpha=@n26, Translation="0,0,0", Rotation="5,0,0", Scale="1,1,1") {
    rotationMode=BMM_Additive
    rotationSpace=BCS_BoneSpace
    alphaBoolBlend {blendOption=Linear}
    boneToModify {boneName=dyn_hair_strap_a_base_l}
    rotation {pitch=5}
   }
   n28 = "Transform (Modify) Bone - Bone: dyn_hair_strap_a_base_r"(ComponentPose=@n27, Alpha=@n26, Translation="0,0,0", Rotation="5,0,0", Scale="1,1,1") {
    rotationMode=BMM_Additive
    rotationSpace=BCS_BoneSpace
    alphaBoolBlend {blendOption=Linear}
    boneToModify {boneName=dyn_hair_strap_a_base_r}
    rotation {pitch=5}
   }
   n29 = Look At - Bone: dyn_hair_strap_d_l(ComponentPose=@n28, Alpha=1) {
    alphaBoolBlend {blendOption=Linear}
    boneToModify {boneName=dyn_hair_strap_d_l}
    lookAtLocation=<-20 -17.4598 -5>
    lookAtTarget {boneReference {boneName=spine_02}}
    lookAt_Axis {axis=<1 0>}
   }
   n30 = Look At - Bone: dyn_hair_strap_d_r(ComponentPose=@n29, Alpha=1) {
    alphaBoolBlend {blendOption=Linear}
    boneToModify {boneName=dyn_hair_strap_d_r}
    lookAtLocation=<-20 -20.1994 5>
    lookAtTarget {boneReference {boneName=spine_02}}
    lookAt_Axis {axis=<1 0>}
   }
   n31 = Get NotCrouching() -> (NotCrouching:bool)
   n32 = Look At - Bone: dyn_hair_strap_l_l!(ComponentPose=@n30, bAlphaBoolEnabled:bool=@n31) -> (Pose) {
    alphaInputType=Bool
    alphaBoolBlend {blendOption=Linear}
    boneToModify {boneName=dyn_hair_strap_l_l}
    lookAtLocation=<-20 -20.862 -5>
    lookAtTarget {boneReference {boneName=spine_02}}
    lookAt_Axis {axis=<1 0>}
   }
   n33 = Look At - Bone: dyn_hair_strap_r_r!(ComponentPose=@n32, bAlphaBoolEnabled:bool=@n31) -> (Pose) {
    alphaInputType=Bool
    alphaBoolBlend {blendOption=Linear}
    boneToModify {boneName=dyn_hair_strap_r_r}
    lookAtLocation=<-20 -22.4995 5>
    lookAtTarget {boneReference {boneName=spine_02}}
    lookAt_Axis {axis=<1 0>}
   }
   n34 = Get BwdMovement() -> (BwdMovement:double)
   n35 = Get TrailControllerFakeVelocity() -> (TrailControllerFakeVelocity:Vector)
   n36 = Trail controller - Bone: dyn_hair_strap_g_l!(ComponentPose=@n33, Alpha=@n34, FakeVelocity:Vector=@n35) -> (Pose) {
    debugLifeTime=0.1
    maxDeltaTime=0.0333
    relaxationSpeedScale=1.2
    alphaScaleBiasClamp {bInterpResult=true, interpSpeedDecreasing=30}
    baseJoint {boneName=root}
    fakeVelocity=<15 5 -15>
    rotationLimits [{limitMax=<180 180 180>}, {limitMax=<180 180 180>}]
    rotationOffsets [<0 0 0>, <0 0 0>]
    trailBone {boneName=dyn_hair_strap_g_l}
    trailRelaxationSpeed {editorCurveData {keys [
       {arriveTangent=-55.2814, interpMode=RCIM_Cubic, time=0, value=21}
       {arriveTangent=-20.9221, interpMode=RCIM_Cubic, leaveTangent=-97.2961, tangentMode=RCTM_User, time=1, value=10}
      ]}}
   }
   n37 = Trail controller - Bone: dyn_hair_strap_g_r!(ComponentPose=@n36, Alpha=@n34, FakeVelocity:Vector=@n35) -> (Pose) {
    debugLifeTime=0.1
    maxDeltaTime=0.0333
    relaxationSpeedScale=1.4
    alphaScaleBiasClamp {bInterpResult=true, interpSpeedDecreasing=30}
    baseJoint {boneName=root}
    fakeVelocity=<10 5 -10>
    rotationLimits [{limitMax=<180 180 180>}, {limitMax=<180 180 180>}]
    rotationOffsets [<0 0 0>, <0 0 0>]
    trailBone {boneName=dyn_hair_strap_g_r}
    trailRelaxationSpeed {editorCurveData {keys [
       {arriveTangent=-55.2814, interpMode=RCIM_Cubic, time=0, value=21}
       {arriveTangent=-21.4671, interpMode=RCIM_Cubic, leaveTangent=-117.9233, tangentMode=RCTM_User, time=1, value=10}
      ]}}
   }
   n38 = Get Jumping() -> (Jumping:bool)
   n39 = Look At - Bone: dyn_hair_strap_l_l(ComponentPose=@n37, bAlphaBoolEnabled=@n38) {
    alphaInputType=Bool
    alphaBoolBlend {blendInTime=2, blendOption=Linear, blendOutTime=2}
    boneToModify {boneName=dyn_hair_strap_l_l}
    lookAtLocation=<-20 -20.862 -5>
    lookAtTarget {boneReference {boneName=spine_02}}
    lookAt_Axis {axis=<1 0>}
   }
   n40 = Look At - Bone: dyn_hair_strap_r_r(ComponentPose=@n39, bAlphaBoolEnabled=@n38) {
    alphaInputType=Bool
    alphaBoolBlend {blendInTime=2, blendOption=Linear, blendOutTime=2}
    boneToModify {boneName=dyn_hair_strap_r_r}
    lookAtLocation=<-20 -22.4995 5>
    lookAtTarget {boneReference {boneName=spine_02}}
    lookAt_Axis {axis=<1 0>}
   }
   n41 = "Transform (Modify) Bone - Bone: dyn_hair_strap_l_l"!(ComponentPose=@n40, bAlphaBoolEnabled:bool=True, Translation="0,0,0", Rotation="15,0,0", Scale="1,1,1") -> (Pose) {
    alphaInputType=Bool
    rotationMode=BMM_Additive
    rotationSpace=BCS_BoneSpace
    alphaBoolBlend {blendInTime=0.1, blendOption=Linear, blendOutTime=1}
    boneToModify {boneName=dyn_hair_strap_l_l}
    rotation {pitch=15}
   }
   n42 = "Transform (Modify) Bone - Bone: dyn_hair_strap_r_r"!(ComponentPose=@n41, bAlphaBoolEnabled:bool=True, Translation="0,0,0", Rotation="15,0,0", Scale="1,1,1") -> (Pose) {
    alphaInputType=Bool
    rotationMode=BMM_Additive
    rotationSpace=BCS_BoneSpace
    alphaBoolBlend {blendInTime=0.1, blendOption=Linear, blendOutTime=1}
    boneToModify {boneName=dyn_hair_strap_r_r}
    rotation {pitch=15}
   }
   Output Pose(Result=Blend Poses by bool(
     BlendPose_0=Blend Poses by bool(BlendPose_0=@n17, BlendPose_1=@n21, BlendTime_0=0.3, BlendTime_1=0.3, bActiveValue=@n22) {blendType=Linear}
     BlendPose_1=Component To Local(ComponentPose=@n42)
     BlendTime_0=0.3
     BlendTime_1=0.3
     bActiveValue=Get bRigidBodyEnabled()
    ) {blendType=Linear}) {layerGroup=None}
   n43 = Anim Dynamics - Bone: dyn_hairTips_01!(
    ComponentPose=Local To Component(LocalPose=Copy Pose From Mesh() {bUseAttachedParent=true})
    Alpha=1
    GravityScale:float=1
    GravityOverride:Vector="0,0,-980"
    ExternalForce="0,0,0"
   ) -> (Pose) {
    angularDampingOverride=0.9999
    angularSpringConstant=1000
    bAngularSpring=true
    bGravityOverrideInSimSpace=true
    bOverrideAngularDamping=true
    simulationSpace=RootRelative
    alphaBoolBlend {blendOption=Linear}
    boundBone {boneName=dyn_hairTips_01}
    componentAppliedLinearAccClamp=<5000 5000 5000>
    componentLinearAccScale=<0.6 0.6 0.6>
    componentLinearVelScale=<0.6 0.6 0.6>
    gravityOverride {z=-980}
    physicsBodyDefinitions [{
      sphereCollisionRadius=0
      boundBone {boneName=dyn_hairTips_01}
      boxExtents=<12 10 10>
      constraintSetup {angularLimitsMax=<0 0 20>, angularLimitsMin=<0 -20 -20>, angularTarget=<0 0 0>}
      localJointOffset=<1.5 0 0>
     }]
    relativeSpaceBone {boneName=head}
   }
   n44 = Anim Dynamics - Bone: dyn_hairTips_02!(ComponentPose=@n43, Alpha=1, GravityScale:float=1, GravityOverride:Vector="0,0,-980", ExternalForce="0,0,0") -> (Pose) {
    angularDampingOverride=1
    angularSpringConstant=1000
    bAngularSpring=true
    bGravityOverrideInSimSpace=true
    bOverrideAngularDamping=true
    simulationSpace=RootRelative
    alphaBoolBlend {blendOption=Linear}
    boundBone {boneName=dyn_hairTips_02}
    componentAppliedLinearAccClamp=<3000 3000 3000>
    componentLinearAccScale=<0.6 0.6 0.6>
    componentLinearVelScale=<0.6 0.6 0.6>
    gravityOverride {z=-980}
    physicsBodyDefinitions [{
      sphereCollisionRadius=0
      boundBone {boneName=dyn_hairTips_02}
      boxExtents=<14 12 10>
      constraintSetup {angularLimitsMax=<0 0 15>, angularLimitsMin=<0 -25 -15>, angularTarget=<0 0 0>}
      localJointOffset=<3 0 0>
     }]
    relativeSpaceBone {boneName=head}
   }
   n45 = Anim Dynamics - Bone: dyn_hairTips_03!(ComponentPose=@n44, Alpha=1, GravityScale:float=1, GravityOverride:Vector="0,0,-980", ExternalForce="0,0,0") -> (Pose) {
    angularDampingOverride=1
    angularSpringConstant=1000
    bAngularSpring=true
    bGravityOverrideInSimSpace=true
    bOverrideAngularDamping=true
    simulationSpace=RootRelative
    alphaBoolBlend {blendOption=Linear}
    boundBone {boneName=dyn_hairTips_03}
    componentAppliedLinearAccClamp=<3000 3000 3000>
    componentLinearAccScale=<0.6 0.6 0.6>
    componentLinearVelScale=<0.6 0.6 0.6>
    gravityOverride {z=-980}
    physicsBodyDefinitions [{
      sphereCollisionRadius=0
      boundBone {boneName=dyn_hairTips_03}
      boxExtents=<14 12 10>
      constraintSetup {angularLimitsMax=<0 0 15>, angularLimitsMin=<0 -25 -15>, angularTarget=<0 0 0>}
      localJointOffset=<3 0 0>
     }]
    relativeSpaceBone {boneName=head}
   }
   n46 = Anim Dynamics - Bone: dyn_hairTips_04!(ComponentPose=@n45, Alpha=1, GravityScale:float=1, GravityOverride:Vector="0,0,-980", ExternalForce="0,0,0") -> (Pose) {
    angularDampingOverride=1
    angularSpringConstant=1000
    bAngularSpring=true
    bGravityOverrideInSimSpace=true
    bOverrideAngularDamping=true
    simulationSpace=RootRelative
    alphaBoolBlend {blendOption=Linear}
    boundBone {boneName=dyn_hairTips_04}
    componentAppliedLinearAccClamp=<5000 5000 5000>
    componentLinearAccScale=<0.6 0.6 0.6>
    componentLinearVelScale=<0.6 0.6 0.6>
    gravityOverride {z=-980}
    physicsBodyDefinitions [{
      sphereCollisionRadius=0
      boundBone {boneName=dyn_hairTips_04}
      boxExtents=<12 10 10>
      constraintSetup {angularLimitsMax=<0 0 20>, angularLimitsMin=<0 -20 -20>, angularTarget=<0 0 0>}
      localJointOffset=<1.5 0 0>
     }]
    relativeSpaceBone {boneName=head}
   }
   n47 = Anim Dynamics - Bone: dyn_hairTips_05!(ComponentPose=@n46, Alpha=1, GravityScale:float=1, GravityOverride:Vector="0,0,-980", ExternalForce="0,0,0") -> (Pose) {
    angularDampingOverride=0.9999
    angularSpringConstant=1000
    bAngularSpring=true
    bGravityOverrideInSimSpace=true
    bOverrideAngularDamping=true
    simulationSpace=RootRelative
    alphaBoolBlend {blendOption=Linear}
    boundBone {boneName=dyn_hairTips_05}
    componentAppliedLinearAccClamp=<5000 5000 5000>
    componentLinearAccScale=<0.6 0.6 0.6>
    componentLinearVelScale=<0.6 0.6 0.6>
    gravityOverride {z=-980}
    physicsBodyDefinitions [{
      sphereCollisionRadius=0
      boundBone {boneName=dyn_hairTips_05}
      boxExtents=<12 10 10>
      constraintSetup {angularLimitsMax=<0 5 15>, angularLimitsMin=<0 -20 -15>, angularTarget=<0 0 0>}
      localJointOffset=<1.5 0 0>
     }]
    relativeSpaceBone {boneName=head}
   }
   n48 = Anim Dynamics - Bone: dyn_hairTips_06!(ComponentPose=@n47, Alpha=1, GravityScale:float=1, GravityOverride:Vector="0,0,-980", ExternalForce="0,0,0") -> (Pose) {
    angularDampingOverride=0.9999
    angularSpringConstant=1000
    bAngularSpring=true
    bGravityOverrideInSimSpace=true
    bOverrideAngularDamping=true
    simulationSpace=RootRelative
    alphaBoolBlend {blendOption=Linear}
    boundBone {boneName=dyn_hairTips_06}
    componentAppliedLinearAccClamp=<5000 5000 5000>
    componentLinearAccScale=<0.6 0.6 0.6>
    componentLinearVelScale=<0.6 0.6 0.6>
    gravityOverride {z=-980}
    physicsBodyDefinitions [{
      sphereCollisionRadius=0
      boundBone {boneName=dyn_hairTips_06}
      boxExtents=<10 15 15>
      constraintSetup {angularLimitsMax=<0 0 15>, angularLimitsMin=<0 -20 -15>, angularTarget=<0 0 0>}
      localJointOffset=<2 0 0>
     }]
    relativeSpaceBone {boneName=head}
   }
   n49 = Anim Dynamics - Bone: dyn_hairTips_07!(ComponentPose=@n48, Alpha=1, GravityScale:float=1, GravityOverride:Vector="0,0,-980", ExternalForce="0,0,0") -> (Pose) {
    angularDampingOverride=1
    angularSpringConstant=1000
    bAngularSpring=true
    bGravityOverrideInSimSpace=true
    bOverrideAngularDamping=true
    simulationSpace=RootRelative
    alphaBoolBlend {blendOption=Linear}
    boundBone {boneName=dyn_hairTips_07}
    componentAppliedLinearAccClamp=<3000 3000 3000>
    componentLinearAccScale=<0.6 0.6 0.6>
    componentLinearVelScale=<0.6 0.6 0.6>
    gravityOverride {z=-980}
    physicsBodyDefinitions [{
      sphereCollisionRadius=0
      boundBone {boneName=dyn_hairTips_07}
      boxExtents=<14 12 10>
      constraintSetup {angularLimitsMax=<0 5 15>, angularLimitsMin=<0 -25 -15>, angularTarget=<0 0 0>}
      localJointOffset=<3 0 0>
     }]
    relativeSpaceBone {boneName=head}
   }
   n50 = Anim Dynamics - Bone: dyn_hairTips_08!(ComponentPose=@n49, Alpha=1, GravityScale:float=1, GravityOverride:Vector="0,0,-980", ExternalForce="0,0,0") -> (Pose) {
    angularDampingOverride=1
    angularSpringConstant=1000
    bAngularSpring=true
    bGravityOverrideInSimSpace=true
    bOverrideAngularDamping=true
    simulationSpace=RootRelative
    alphaBoolBlend {blendOption=Linear}
    boundBone {boneName=dyn_hairTips_08}
    componentAppliedLinearAccClamp=<5000 5000 5000>
    componentLinearAccScale=<0.6 0.6 0.6>
    componentLinearVelScale=<0.6 0.6 0.6>
    gravityOverride {z=-980}
    physicsBodyDefinitions [{
      sphereCollisionRadius=0
      boundBone {boneName=dyn_hairTips_08}
      boxExtents=<12 10 10>
      constraintSetup {angularLimitsMax=<0 0 20>, angularLimitsMin=<0 -20 -20>, angularTarget=<0 0 0>}
      localJointOffset=<1 0 -2>
     }]
    relativeSpaceBone {boneName=head}
   }
   n51 = Look At - Bone: dyn_hair_strap_a_base_l(ComponentPose=@n50, Alpha=1) {
    alphaBoolBlend {blendOption=Linear}
    boneToModify {boneName=dyn_hair_strap_a_base_l}
    lookAtLocation=<0 -20 -3.5>
    lookAtTarget {boneReference {boneName=spine_04}}
    lookAt_Axis {axis=<1 0>}
   }
   n52 = Look At - Bone: dyn_hair_strap_a_base_r(ComponentPose=@n51, Alpha=1) {
    alphaBoolBlend {blendOption=Linear}
    boneToModify {boneName=dyn_hair_strap_a_base_r}
    lookAtLocation=<0 -20 3.2>
    lookAtTarget {boneReference {boneName=spine_04}}
    lookAt_Axis {axis=<1 0>}
   }
   Save cached pose 'PerPose'(Pose=Component To Local(ComponentPose=@n52)) {cacheName=PerPose}
  ]}
 EventGraph {nodes [
   Event BlueprintUpdateAnimation()
   > n1 = Sequence!(execute) -> (then_0, then_1, then_2:Exec, then_3:Exec, then_4:Exec, then_5:Exec, then_6:Exec)
   > n2 = Gravity Override State Function DEPRECATED(
    CharacterPartAnimInstance=Self-Reference()
    GameTimeInSeconds=GetGameTimeInSeconds()
    GravityOverrideParams=Get Gravity Override Params()
    DebugJointName=head
   )
   > Set GravityOverride(GravityOverride=@n2)
   > n3 = Random Vector Function!(
    execute
    CharacterPartAnimInstance:Custom Character Part Anim Instance Object Reference=Self-Reference()
    SkelMeshComponent:Mesh Component Object Reference
    JointName:Name=None
    GameTimeInSeconds:double=GetGameTimeInSeconds()
    TimeOffset:double=0
    PerlinScale:double=1
    Frequency:Vector="1,1,2"
    Amplitude:Vector="15,20,30"
    Offset:Vector="0,-50,30"
   ) -> (then, VectorOut:Vector)
   > Set TrailControllerFakeVelocity(TrailControllerFakeVelocity:Vector=@n3) -> (Output_Get:Vector)
   > n4 = Branch(Condition=Get bIsTargeting())
   > Set AccScaleOverride(AccScaleOverride:Vector="0.1,0.1,0") -> (Output_Get:Vector)
   Set AccScaleOverride(execute=@n4.else, AccScaleOverride="0.1,0.2,0.2")
   n5 = PlayerMovement(execute=@n1.then_2, AnimInstance=Self-Reference(), XDotProductCoefficient=1, YDotProductCoefficient=1, ZDotProductCoefficient=1)
   n6 = MapRangeClamped(Value=@n5.YMovement, InRangeA=0, InRangeB=-1)
   n7 = MapRangeClamped(Value="Absolute (Float)"(A=@n5.ZMovement), InRangeA=0, InRangeB=1, OutRangeB=0.2)
   > Set BwdMovement(BwdMovement:double=MapRangeClamped(Value=float + float(A=@n7, B=@n6) {is=title K2Node_CommutativeAssociativeBinaryOperator}, InRangeA=0, InRangeB=1.2, OutRangeA=1, OutRangeB=0)) -> (Output_Get:double)
   > Branch(Condition=Get bIsSkyDiving())
   > Set BwdMovement(BwdMovement=0)
   n8 = Get!(execute:Exec=@n1.then_3) -> (then:Exec, else:Exec, playerPawn:Fort Player Pawn Object Reference)
   > Do Once()
   n9 = GetSkeletalMeshForPartType(self=@n8, PartType=Backpack)
   > Set isRumbleBackbling(isRumbleBackbling:bool=MatchesWildcard(SourceString=GetDisplayName(Object=@n9), Wildcard=*Rumble_*)) -> (Output_Get:bool)
   n10 = Branch(execute=@n1.then_5, Condition=Get bIsCrouching())
   > Set NotCrouching(NotCrouching:bool=false) -> (Output_Get:bool)
   Set NotCrouching(execute=@n10.else, NotCrouching=true)
   n11 = Branch(execute=@n1.then_6, Condition=OR Boolean(A=Get bIsJumping(), B=Get bIsFalling()))
   > Set Jumping(Jumping:bool=true) -> (Output_Get:bool)
   Set Jumping(execute=@n11.else, Jumping=false)
   n12 = GetGameTimeInSeconds()
   n13 = RandomRotatorFunction!(
    execute=@n1.then_1
    GameTimeInSeconds:double=@n12
    PerlinScale:double=0.2
    Frequency:Rotator="1.5,1.8,0"
    Amplitude:Rotator="10,5,0"
    Offset:Rotator="10,0, 0"
    TimeOffset:double=0
   ) -> (then, RandomRotator:Rotator)
   > Set HairRotator_a_01(HairRotator_a_01:Rotator=@n13) -> (Output_Get:Rotator)
   > n14 = RandomRotatorFunction(GameTimeInSeconds=@n12, PerlinScale=0.2, Frequency="1,1, 0", Amplitude="10,5, 0", Offset="10,0, 0", TimeOffset=0)
   > Set HairRotator_a_02(HairRotator_a_02:Rotator=@n14) -> (Output_Get:Rotator)
   > n15 = RandomRotatorFunction(GameTimeInSeconds=@n12, PerlinScale=0.2, Frequency="3,1,0", Amplitude="5,2,0", Offset="15, 0,0", TimeOffset=0)
   > Set HairRotator_a_03(HairRotator_a_03:Rotator=@n15) -> (Output_Get:Rotator)
   > n16 = RandomRotatorFunction(GameTimeInSeconds=@n12, PerlinScale=0.2, Frequency="1,0.2, 0", Amplitude="5,5, 0", Offset="15, 0, 0", TimeOffset=0)
   > Set HairRotator_a_04(HairRotator_a_04:Rotator=@n16) -> (Output_Get:Rotator)
   n17 = GetGameTimeInSeconds()
   > n18 = RandomRotatorFunction(GameTimeInSeconds=@n17, PerlinScale=0.5, Frequency="6,6,0", Amplitude="15,10,0", Offset="10,0, 0", TimeOffset=0)
   > Set HairRotator_01(HairRotator_01:Rotator=@n18) -> (Output_Get:Rotator)
   > n19 = RandomRotatorFunction(GameTimeInSeconds=@n17, PerlinScale=0.5, Frequency="6,6, 0", Amplitude="15,16,0", Offset="10,0, 0", TimeOffset=0)
   > Set HairRotator_02(HairRotator_02:Rotator=@n19) -> (Output_Get:Rotator)
   > n20 = RandomRotatorFunction(GameTimeInSeconds=@n17, PerlinScale=0.5, Frequency="12,8,0", Amplitude="12,6,0", Offset="18, 0,0", TimeOffset=0)
   > Set HairRotator_03(HairRotator_03:Rotator=@n20) -> (Output_Get:Rotator)
   > n21 = RandomRotatorFunction(GameTimeInSeconds=@n17, PerlinScale=0.5, Frequency="5,2, 0", Amplitude="14,10,0", Offset="20, 0,0", TimeOffset=0)
   > Set HairRotator_04(HairRotator_04:Rotator=@n21) -> (Output_Get:Rotator)
   n22 = GetOwnerSkeletalMeshComponent()
   n23 = BreakRotator(InRot=BreakTransform(InTransform=GetDeltaTransformFromRefPose(self=@n22, BoneName=head, BaseName=neck_02)).Rotation)
   Set LookUp(execute=@n1.then_4, LookUp:double=MapRangeClamped(Value=@n23.Yaw, InRangeA=5, InRangeB=-12)) -> (Output_Get:double)
   > Set LookDown(LookDown:double=MapRangeClamped(Value=@n23.Yaw, InRangeA=0, InRangeB=15)) -> (Output_Get:double)
   Event OnCharacterCustomizationLoaded()
   > Do Once()
   n24 = CheckForBackbling()
   > Set isWearBackbling(isWearBackbling:bool=@n24) -> (Output_Get:bool)
   > Set isWearCape(isWearCape:bool=@n24.bIsBackblingLong) -> (Output_Get:bool)
   Event BlueprintInitializeAnimation()
   n25 = TryGetPawnOwner()
   > Is Valid(InputObject=@n25)
   > n26 = Cast To FortPlayerPawnAthena(Object=@n25)
   > Set playerPawn(playerPawn:Fort Player Pawn Object Reference=@n26) -> (Output_Get:Fort Player Pawn Object Reference)
   RandomRotatorFunction(GameTimeInSeconds=0, PerlinScale=1, Frequency="3,1,0", Amplitude="25,15,0", Offset="15, 0,0", TimeOffset=0)
   RandomRotatorFunction(GameTimeInSeconds=0, PerlinScale=1, Frequency="2.2,0.8, 0", Amplitude="25,15, 0", Offset="15, 0, 0", TimeOffset=0)
   RandomRotatorFunction(GameTimeInSeconds=0, PerlinScale=1, Frequency="2.8,1.8,0", Amplitude="40,15,0", Offset="10,0, 0", TimeOffset=0)
   RandomRotatorFunction(GameTimeInSeconds=0, PerlinScale=1, Frequency="1.8,1.5, 0", Amplitude="40,15, 0", Offset="10,0, 0", TimeOffset=0)
  ]}
]
variables [
 AccScaleOverride {type=Vector}
 BwdMovement {type=number}
 currentBackpack {type=AthenaBackpackItemDefinition, default=null}
 Gravity Override Params {
  type=GravityOverrideParamsStruct
  default {
   emoteOrMelee {gravityOverride=<0 -50 -15>, windAmplitude=<50 -25 100>, windFrequency=<1.5 10 4>, windOffset=<0 -100 -50>}
   front_End {gravityOverride=<0 -50 -15>, windAmplitude=<50 -25 60>, windFrequency=<1.5 10 4>, windOffset=<0 -100 -50>}
   onGround_Moving {gravityOverride=<0 -50 -45>, windAmplitude=<25 -25 45>, windFrequency=<0.5 10 4>, windOffset=<50 -80 -60>}
   onGround_Standing {gravityOverride=<0 -50 -30>, windAmplitude=<25 -35 65>, windFrequency=<1 10 5>, windOffset=<60 -100 -60>}
   skydive_FloatingDown {gravityOverride=<0 -700 300>, windAmplitude=<200 -500 200>, windOffset=<0 -600 150>}
   skydive_Parachuting {gravityOverride=<0 -50 -300>, windAmplitude=<25 -25 50>, windFrequency=<1.5 6 8>, windOffset=<80 -80 -25>}
  }
 }
 GravityOverride {type=Vector}
 HairRotator_01 {type=Rotator}
 HairRotator_02 {type=Rotator}
 HairRotator_03 {type=Rotator}
 HairRotator_04 {type=Rotator}
 HairRotator_a_01 {type=Rotator}
 HairRotator_a_02 {type=Rotator}
 HairRotator_a_03 {type=Rotator}
 HairRotator_a_04 {type=Rotator}
 isRumbleBackbling {type=boolean}
 isWearBackbling {type=boolean}
 isWearCape {type=boolean}
 Jumping {type=boolean}
 LookDown {type=number}
 LookUp {type=number}
 NotCrouching {type=boolean}
 playerPawn {type=FortPlayerPawn}
 rumbleBackblingSkelMesh {type=SkeletalMesh, default=null}
 TrailControllerFakeVelocity {type=Vector}
]
